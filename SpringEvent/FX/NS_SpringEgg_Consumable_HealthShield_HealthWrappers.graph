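# Niagara system for the health variant of the Spring Event egg health/shield consumable, throwing short-lived green mesh wrapper pieces.
# One emitter spawns randomly scaled and oriented meshes from a raised sphere, then gravity, curl noise and aerodynamic drag make them tumble.
# A region mask and lifetime curve drive mesh scale.
# The skeletal-mesh spawn location and purple-blue initializer are disabled, and the isShielding user variable has no visible use.
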NiagaraSystem /SpringEvent/FX/NS_SpringEgg_Consumable_HealthShield_HealthWrappers
dependencies {
 usedDataInterfaces [&NiagaraDataInterfaceArrayColor, &NiagaraDataInterfaceCurve, &NiagaraDataInterfaceColorCurve]
 usedDynamicInputs []
 usedModules [&SystemState]
}
emitters {Wrappings001 {
  emitterUpdateScript [
   &EmitterState(Life Cycle Mode=::NewEnumerator1, Loop Behavior=::NewEnumerator1, Loop Delay=0.07, Loop Duration=0.2, UseLoopDelay=false)
   &SpawnRate(SpawnRate=100, Use Spawn Probability=false)
  ]
  particleSpawnScript [
   &InitializeParticle(
    Color=&SelectLinearColorFromArray(
     Array Sampling Mode=::NewEnumerator0
     Direct Array Index=?
     Color Selection Array=DataInterface() {propertyValues {ColorData [<rgba 0 0.8542 0.104 1>, <rgba 0 0.7188 0.3468 1>, <rgba 0 0.8646 0.4517 1>]}}
     Random Seed=0
    )
    Lifetime Max=0.6
    Lifetime Min=0.3
    Lifetime Mode=::NewEnumerator1
    Mesh Scale Mode=::NewEnumerator1
    Mesh Uniform Scale Max=0.5
    Mesh Uniform Scale Min=0.2
    Sprite Rotation Mode=::NewEnumerator2
    Sprite UV Mode=::NewEnumerator3
    UsePositionOffset=false
    Write Material Random=true
   )
   &InitializeParticle(
    Color=&SelectLinearColorFromArray(
     Array Sampling Mode=::NewEnumerator0
     Direct Array Index=?
     Color Selection Array=DataInterface() {propertyValues {ColorData [<rgba 0.4038 0.1928 0.8542 1>, <rgba 0.179 0.1428 0.7188 1>, <rgba 0.1449 0.3188 0.8646 1>]}}
     Random Seed=0
    )
    Lifetime Max=0.6
    Lifetime Min=0.4
    Lifetime Mode=::NewEnumerator1
    Mesh Scale Mode=::NewEnumerator1
    Mesh Uniform Scale Max=0.3
    Mesh Uniform Scale Min=0.15
    Sprite Rotation Mode=::NewEnumerator2
    Sprite UV Mode=::NewEnumerator3
    UsePositionOffset=false
    Write Material Random=true
   ) {enabled=false, module=InitializeParticle001}
   &SkeletalMeshLocation(Bone Sampling Mode=::NewEnumerator4, MaskSpawnGroup=false, Skeletal Mesh=$User.SkeletalMesh) {enabled=false}
   &ShapeLocation(MaskSpawnGroup=false, Offset=<0 0 70>, Offset Mode=::NewEnumerator0, Override Seed=false, Sphere Radius=25, UseOwnerScale=true) {module=ShapeLocation001}
   &RegionMask(Offset:Vector3f=<0 0 15>, Region Falloff=0.4, Region Radius=45, UseOffset=true, UseRotation=false)
   &AddVelocity(
    Module.Use Falloff Distance=false
    Override Seed=false
    Velocity Mode=::NewEnumerator1
    Velocity Speed=&RandomRangeFloat(Minimum=70, Maximum=180)
   )
   &InitialMeshOrientation(
    Mesh Orientation Mode=::NewEnumerator2
    Rotation=&RandomRangeVector(Minimum=<0 0 0>, Maximum=<1 1 1>)
    Use Rotation Vector=true
   )
   &DynamicMaterialParameters(
    Index 0 Param 1=0.5
    Index 0 Param 2=0.09
    Index 0 Param 3=0.1
    Index 0 Param 4=0.3
    Index 1 Param 2=0.08
    Module.Param0WriteEnabled=true
    Module.Param0WriteEnabled1=true
    Module.Param1WriteEnabled=true
    Module.Param1WriteEnabled1=true
    Module.Param2WriteEnabled=true
    Module.Param2WriteEnabled1=false
    Module.Param3WriteEnabled=true
    Module.Param3WriteEnabled1=false
    Write Parameter Index 1=true
   )
  ]
  particleUpdateScript [
   &ParticleState()
   &ScaleMeshSize(Scale Factor=&VectorFromFloat(Value=&Lerp_Float(A=1, B=0, Alpha=$StackContext.RegionMask.Float))) {module=ScaleMeshSize001}
   &ScaleMeshSize(Scale Factor=&VectorFromFloat(Value=&FloatFromCurve(
      FloatCurve=DataInterface() {propertyValues {Curve {keys [
          <at 0 0 RCIM_Cubic>
          {arriveTangent=0.0142, interpMode=RCIM_Cubic, leaveTangent=0.0142, tangentMode=RCTM_User, time=0.12, value=0.5}
          {arriveTangent=-0.0476, interpMode=RCIM_Cubic, leaveTangent=-0.0476, tangentMode=RCTM_User, time=0.45, value=0.5}
          <at 1 0 RCIM_Cubic>
         ]}}}
      Scale Curve=0.4
     )))
   &GravityForce(Coordinate Space=::Simulation, Gravity=<0 0 -300>)
   &CurlNoiseForce(
    Noise Frequency=&RandomRangeFloat(Minimum=20, Maximum=30)
    Noise Strength=&FloatFromCurve(FloatCurve=DataInterface() {propertyValues {Curve {keys [<at 0 0.5>,<at 1 1>]}}}, Scale Curve=300)
    Pan Noise=false
   )
   &ScaleColor(
    Scale RGB=&MakeVectorFromLinearColorRGB(Color=&ColorFromCurve(ColorCurve=DataInterface() {propertyValues {
        AlphaCurve {keys [<at 0 1 RCIM_Cubic>]}
        BlueCurve {keys [<at 0.0397 2 RCIM_Cubic>,<at 0.2164 1>,<at 0.9066 0.4115 RCIM_Cubic>]}
        GreenCurve {keys [<at 0.0397 2 RCIM_Cubic>,<at 0.2164 1>,<at 0.9066 0.4115 RCIM_Cubic>]}
        RedCurve {keys [<at 0.0397 2 RCIM_Cubic>,<at 0.2164 1>,<at 0.9066 0.4115 RCIM_Cubic>]}
       }}))
    ScaleA=true
    ScaleRGB=true
   ) {enabled=false}
   &AerodynamicDrag(
    Aerodynamic Drag=0.2
    Aerodynamic Rotational Drag=1.5
    Drag Shape Curve=DataInterface() {propertyValues {Curve {keys [
        {interpMode=RCIM_Cubic, tangentMode=RCTM_User, time=0, value=0.1}
        {arriveTangent=2.0799, interpMode=RCIM_Cubic, leaveTangent=2.0799, tangentMode=RCTM_User, time=1, value=1}
       ]}}}
    Lift Shape Curve=DataInterface() {propertyValues {Curve {keys [
        {interpMode=RCIM_Cubic, tangentMode=RCTM_User, time=0, value=0}
        {interpMode=RCIM_Cubic, tangentMode=RCTM_User, time=0.5, value=1}
        {tangentMode=RCTM_User, time=1, value=0}
       ]}}}
    Use Linear Drag=true
    Use Rotational Drag=true
   )
   &SolveForcesAndVelocity(Clamp Velocity=false, Limit Acceleration=false)
   &AlignSpriteToMeshOrientation(SetAlignment=true, SetFacing=true)
   &UpdateMeshOrientation(Rotation Rate=&RandomRangeFloat(Minimum=-1, Maximum=1), Rotation Vector=<0 0 1>)
  ]
  rendererClasses [&NiagaraMeshRendererProperties]
  renderers [{rendererClass=&NiagaraMeshRendererProperties, rendererIndex=0}]
 }}
summary {
 systemName=NS_SpringEgg_Consumable_HealthShield_HealthWrappers
 emitters [{emitterName=Wrappings001, rendererClasses [&NiagaraMeshRendererProperties]}]
 userVariables [
  User.isShielding {type=NiagaraBool}
  User.SkeletalMesh {
   type=NiagaraDataInterfaceSkeletalMesh
   defaultValue {
    struct=NiagaraExt_VariableValue_DataInterface
    value {dataInterface=&self:NiagaraDataInterfaceSkeletalMesh_0}
   }
  }
 ]
}
